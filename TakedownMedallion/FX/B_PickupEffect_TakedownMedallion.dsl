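; Pickup effect Blueprint for the Takedown Medallion, handling its sound, VFX and spin as a dropped item.
; On BeginPlay it reads the item definition and, when item data is found, starts a looping sound indicator with that item's dropped-loop sound.
; It then tints the idle Niagara effect by pickup rarity color.
; It also adds a rotating movement component, ticking during physics, that drives the parent pickup's static mesh.
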(event EventBeginPlay
  (bind _parentpickupactor (Variables|PickupProperties|GetParentPickupActor))
  (|CallParentFunction)
  (bind _itemdefinition (Variables|PickupProperties|GetItemDefinition)
    (:then
      (bind _itemdata (Class|ItemDefinitionBase|GetItemData _itemdefinition)
        (:Found
          (SoundIndicator|StartSoundIndicatorLooping (Variables|Default|GetFortSoundIndicator) (Transformation|GetActorLocation) (Utilities|ResolveSoftReference (.DroppedLoopSound (Utilities|Struct|BreakFortnitePickupItemData _itemdata))) 1.0 0))
        (:NotFound
          _))
      (Niagara|SetNiagaraVariable(LinearColor) (Variables|Default|GetNS_TDM_Pickup_Idle) "RarityColor" (Variables|FortPickupsParent|GetPickupRarityColor))
      (bind _returnvalue (AddComponent|NODE_AddRotatingMovementComponent-0 :self _parentpickupactor))
      (Components|Tick|SetTickGroup _returnvalue "TG_DuringPhysics")
      (Components|Movement|SetUpdatedComponent _returnvalue (Class|FortPickup|GetPickupStaticMeshComponent _parentpickupactor)))
    (:else
      _)))

(fn ConstructionScript
  (|CallParentFunction))
